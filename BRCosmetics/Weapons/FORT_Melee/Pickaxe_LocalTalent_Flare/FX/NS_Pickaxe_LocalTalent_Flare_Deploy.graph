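# Niagara deploy effect for the LocalTalent Flare pickaxe cosmetic, layering mesh music lines, rings with lights and glow sprites.
# Staggered emitter delays build the lines first, then rings and glows, with curves animating material reveals and brightness flashes.
# A system moving average drives line materials and light radius, blended toward fixed values by MusicReactivity.
# User.VariantSwitch indexes color and material arrays, while User.bIsFrontEndPreview swaps one line's material curve.
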NiagaraSystem /BRCosmetics/Weapons/FORT_Melee/Pickaxe_LocalTalent_Flare/FX/NS_Pickaxe_LocalTalent_Flare_Deploy
dependencies {
 usedDataInterfaces [
  &NiagaraDataInterfaceCurve
  &NiagaraDataInterfaceArrayColor
  &NiagaraDataInterfaceArrayFloat3
  &NiagaraDataInterfaceArrayFloat
  &NiagaraDataInterfaceColorCurve
 ]
 usedDynamicInputs []
 usedModules [
  &self:SystemSpawnScript.SystemScriptSource.SystemScriptGraph.NiagaraNodeAssignment_0.SetVariables_#1
  &SystemState
  &InterpolateOverTime
 ]
}
emitters {
 MusicLine {
  emitterUpdateScript [
   &EmitterState(
    Inactive Response=::NewEnumerator1
    Life Cycle Mode=::NewEnumerator1
    Loop Behavior=::NewEnumerator1
    Loop Delay=0.2
    Loop Duration=0.7
    UseLoopDelay=true
   )
   &SpawnBurst_Instantaneous(Spawn Count=2, Use Spawn Probability=true, UseLoopCountLimit=false)
  ]
  particleSpawnScript [
   &PartitionParticles()
   &InitializeParticle(
    Lifetime=5
    Mesh Scale=&Multiply_VectorByFloat(Vector=<2.8 1 1>, Float=1)
    Mesh Scale Mode=::NewEnumerator4
    Position Offset=&SelectVectorFromArray(
     Direct Array Index=$Particles.PartitionParticles.Partition
     Vector Selection Array=DataInterface() {propertyValues {FloatData [<-2 0 15>, <-2 0 -15>]}}
    )
    UsePositionOffset=true
    Write Material Random=true
   )
   &InitialMeshOrientation(
    Mesh Orientation Mode=::NewEnumerator2
    Rotation=&SelectVectorFromArray(
     Direct Array Index=$Particles.PartitionParticles.Partition
     Vector Selection Array=DataInterface() {propertyValues {FloatData [<0 -0.25 0>, <0 0.25 0>]}}
    )
    Use Rotation Vector=true
   )
   &InitialMeshOrientation(
    Mesh Orientation Mode=::NewEnumerator2
    Rotation=&SelectVectorFromArray(
     Direct Array Index=$Particles.PartitionParticles.Partition
     Vector Selection Array=DataInterface() {propertyValues {FloatData [<0 0 0>, <0 0 0.5>]}}
    )
    Use Rotation Vector=true
   ) {module=InitialMeshOrientation001}
  ]
  particleUpdateScript [
   &ParticleState()
   &DynamicMaterialParameters(
    Index 0 Param 1=&Lerp_Float(
     A=&Lerp_Float(A=0.6, B=1.8, Alpha=&ClampFloat(Float=$System.InterpolateOverTime.InterpolateOverTime_Float.Moving Average))
     B=1.5
     Alpha=$System.MusicReactivity
    )
    Index 0 Param 4=&MakeCustomFloatFromBool(
     Bool=$User.bIsFrontEndPreview
     True Float=&FloatFromCurve(
      FloatCurve=DataInterface() {propertyValues {Curve {keys [
          {arriveTangent=3.7071, interpMode=RCIM_Cubic, leaveTangent=3.7071, tangentMode=RCTM_User, time=0.05, value=0}
          {arriveTangent=9.1858, interpMode=RCIM_Cubic, leaveTangent=9.1858, tangentMode=RCTM_User, time=0.15, value=1}
         ]}}}
      Scale Curve=100
     )
     False Float=&FloatFromCurve(
      FloatCurve=DataInterface() {propertyValues {Curve {keys [
          {arriveTangent=6.7984, interpMode=RCIM_Cubic, leaveTangent=6.7984, tangentMode=RCTM_User, time=0, value=0}
          {arriveTangent=9.1858, interpMode=RCIM_Cubic, leaveTangent=9.1858, tangentMode=RCTM_User, time=0.15, value=1}
         ]}}}
      Scale Curve=150
     )
    )
    Index 1 Param 1=2
    Index 1 Param 2=2
    Index 1 Param 3=&SelectFloatFromArray(
     Direct Array Index=&MakeIntFromFloat(Float=$User.VariantSwitch)
     Float Selection Array=DataInterface() {propertyValues {FloatData [0,1]}}
    )
    Module.Param0WriteEnabled=true
    Module.Param0WriteEnabled1=true
    Module.Param1WriteEnabled=true
    Module.Param1WriteEnabled1=true
    Module.Param2WriteEnabled=false
    Module.Param2WriteEnabled1=true
    Module.Param3WriteEnabled=true
    Module.Param3WriteEnabled1=false
    Write Parameter Index 1=true
   ) {module=DynamicMaterialParameters001}
   /SetVariables(Particles.Size003:Vector3f=&Multiply_Vector(A=$Particles.Scale, B=<1 1 1.2>)) {isSetParameters=true, module=SetVariables_#2}
   /SetVariables(Particles.Scale_Point:Vector3f=<1 0.1 0.45>) {isSetParameters=true}
   /SetVariables(Particles.Scale_OutsideGlow:Vector3f=<5 1 3>) {isSetParameters=true}
   &ScaleColor(
    Scale Alpha=0.1
    Scale RGB=&MakeVectorFromLinearColorRGB(Color=&SelectLinearColorFromArray(
      Array Sampling Mode=::NewEnumerator1
      Direct Array Index=&MakeIntFromFloat(Float=$User.VariantSwitch)
      Color Selection Array=DataInterface() {propertyValues {ColorData [<rgba 0.0132 0.5 0.3661 1>, <rgba 0.1759 0.5 0 1>]}}
      Random Seed=?
     ))
    ScaleA=true
    ScaleRGB=true
   )
   /SetVariables(Particles.Color002=&MakeLinearColorFromVectorAndFloat(
     "Vector (RGB)"=&MakeVectorFromLinearColorRGB(Color=&Multiply_LinearColorByFloat(
       Linear Color=&SelectLinearColorFromArray(
        Array Sampling Mode=::NewEnumerator1
        Direct Array Index=&MakeIntFromFloat(Float=$User.VariantSwitch)
        Color Selection Array=DataInterface() {propertyValues {ColorData [<rgba 0 0.2031 0.0304 1>, <rgba 0.1579 0.5 0 1>]}}
        Random Seed=?
       )
       Float=&FloatFromCurve(FloatCurve=DataInterface() {propertyValues {Curve {keys [
            {arriveTangent=71.4328, interpMode=RCIM_Cubic, leaveTangent=-0.0375, tangentMode=RCTM_Break, time=0, value=1}
            {arriveTangent=-0.2374, interpMode=RCIM_Cubic, tangentMode=RCTM_Break, time=0.08, value=1}
            <at 0.1 10 RCIM_Cubic>
            <at 0.15 1 RCIM_Cubic>
           ]}}})
      ))
     "Float (Alpha)"=2
    )) {isSetParameters=true, module=SetVariables_#3}
   /SetVariables(Particles.Color003:LinearColor=&MakeLinearColorFromVectorAndFloat(
     "Vector (RGB)"=&MakeVectorFromLinearColorRGB(Color=&SelectLinearColorFromArray(
       Array Sampling Mode=::NewEnumerator1
       Direct Array Index=&MakeIntFromFloat(Float=$User.VariantSwitch)
       Color Selection Array=DataInterface() {propertyValues {ColorData [<rgba 3 2.0543 0 1>, <rgba 0 2 0.749 1>]}}
       Random Seed=?
      ))
     "Float (Alpha)"=1
    )) {isSetParameters=true, module=SetVariables_#4}
   /SetVariables(Particles.ColorGlow:LinearColor=&MakeLinearColorFromVectorAndFloat(
     "Vector (RGB)"=&MakeVectorFromLinearColorRGB(Color=&SelectLinearColorFromArray(
       Array Sampling Mode=::NewEnumerator1
       Direct Array Index=&MakeIntFromFloat(Float=$User.VariantSwitch)
       Color Selection Array=DataInterface() {propertyValues {ColorData [<rgba 2 1.876 0.3646 1>, <rgba 0.3683 2 0.1059 1>]}}
       Random Seed=?
      ))
     "Float (Alpha)"=0.2
    )) {isSetParameters=true}
   /SetVariables(Particles.ColorPoint:LinearColor=&MakeLinearColorFromVectorAndFloat("Vector (RGB)"=&MakeVectorFromLinearColorRGB(Color=<rgba 0 1 0.6184 1>), "Float (Alpha)"=1)) {isSetParameters=true}
   /SetVariables(Particles.Color_OutsideGlow:LinearColor=&MakeLinearColorFromVectorAndFloat(
     "Vector (RGB)"=&MakeVectorFromLinearColorRGB(Color=&SelectLinearColorFromArray(
       Array Sampling Mode=::NewEnumerator1
       Direct Array Index=&MakeIntFromFloat(Float=$User.VariantSwitch)
       Color Selection Array=DataInterface() {propertyValues {ColorData [<rgba 0 0.5806 0.7083 1>, <rgba 0.3576 0.7083 0.0447 1>]}}
       Random Seed=?
      ))
     "Float (Alpha)"=&Multiply_Float(
      A=0.2
      B=&FloatFromCurve(FloatCurve=DataInterface() {propertyValues {Curve {keys [
           <at 0.07 0 RCIM_Cubic>
           {arriveTangent=71.4328, interpMode=RCIM_Cubic, leaveTangent=-0.0375, tangentMode=RCTM_Break, time=0.0877, value=1}
           {arriveTangent=-0.2374, interpMode=RCIM_Cubic, leaveTangent=41.5138, tangentMode=RCTM_Break, time=0.1, value=1}
           {arriveTangent=169.2863, interpMode=RCIM_Cubic, leaveTangent=-127.0063, tangentMode=RCTM_Break, time=0.116, value=2.5}
           <at 0.15 1 RCIM_Cubic>
          ]}}})
     )
    )) {isSetParameters=true}
  ]
  rendererClasses [&NiagaraMeshRendererProperties]
  renderers [
   {rendererClass=&NiagaraMeshRendererProperties, rendererIndex=0}
   {rendererClass=&NiagaraMeshRendererProperties, rendererIndex=1}
   {rendererClass=&NiagaraMeshRendererProperties, rendererIndex=2}
   {rendererClass=&NiagaraMeshRendererProperties, rendererIndex=3}
   {rendererClass=&NiagaraMeshRendererProperties, rendererIndex=4}
   {rendererClass=&NiagaraMeshRendererProperties, rendererIndex=5}
  ]
 }
 MusicLine001 {
  emitterUpdateScript [
   &EmitterState(
    Inactive Response=::NewEnumerator1
    Life Cycle Mode=::NewEnumerator1
    Loop Behavior=::NewEnumerator1
    Loop Delay=0.2
    Loop Duration=0.7
    UseLoopDelay=true
   )
   &SpawnBurst_Instantaneous(Use Spawn Probability=true, UseLoopCountLimit=true)
  ]
  particleSpawnScript [
   &InitializeParticle(
    Lifetime=1.1
    Mesh Scale=&Multiply_VectorByFloat(Vector=<1 0.1 0.45>, Float=1)
    Mesh Scale Mode=::NewEnumerator4
    Position Offset=<-13 0 0>
    UsePositionOffset=true
    Write Material Random=true
   )
   &InitialMeshOrientation(Mesh Orientation Mode=::NewEnumerator2, Rotation=<0 -0.25 0>, Use Rotation Vector=true)
  ]
  particleUpdateScript [
   &ParticleState()
   &DynamicMaterialParameters(
    Index 0 Param 1=&Multiply_Float(
     A=&FloatFromCurve(FloatCurve=DataInterface() {propertyValues {Curve {keys [<at 0 0>,<at 0.3 1>]}}})
     B=&Lerp_Float(
      A=&Lerp_Float(A=0.5, B=1.5, Alpha=$System.InterpolateOverTime.InterpolateOverTime_Float.Moving Average)
      B=1
      Alpha=$System.MusicReactivity
     )
    )
    Index 0 Param 2=&Multiply_Float(
     A=&FloatFromCurve(FloatCurve=DataInterface() {propertyValues {Curve {keys [<at 0 0>,<at 0.3 1>]}}})
     B=&Lerp_Float(
      A=&Lerp_Float(A=2, B=5, Alpha=$System.InterpolateOverTime.InterpolateOverTime_Float.Moving Average)
      B=3
      Alpha=$System.MusicReactivity
     )
    )
    Index 0 Param 3=1
    Index 1 Param 1=2
    Index 1 Param 2=2
    Module.Param0WriteEnabled=true
    Module.Param0WriteEnabled1=true
    Module.Param1WriteEnabled=true
    Module.Param1WriteEnabled1=true
    Module.Param2WriteEnabled=false
    Module.Param2WriteEnabled1=false
    Module.Param3WriteEnabled=false
    Module.Param3WriteEnabled1=false
    Write Parameter Index 1=true
   ) {module=DynamicMaterialParameters001}
   /SetVariables(Particles.Size003:Vector3f=<1 0.1 0.6>) {isSetParameters=true, module=SetVariables_#2}
   /SetVariables(Particles.Scale_Beam:Vector3f=<0.8 8 0.3>) {isSetParameters=true}
   /SetVariables(Particles.Scale_OutsideGlow:Vector3f=<1.3 1 1.5>) {isSetParameters=true}
   &ScaleColor(
    Scale Alpha=0.15
    Scale RGB=&MakeVectorFromLinearColorRGB(Color=&SelectLinearColorFromArray(
      Array Sampling Mode=::NewEnumerator1
      Direct Array Index=&MakeIntFromFloat(Float=$User.VariantSwitch)
      Color Selection Array=DataInterface() {propertyValues {ColorData [<rgba 0 0.6875 0.4 1>, <rgba 0.4991 0.9 0.081 1>]}}
      Random Seed=?
     ))
    ScaleA=true
    ScaleRGB=true
   )
   /SetVariables(Particles.Color002=&MakeLinearColorFromVectorAndFloat(
     "Vector (RGB)"=&MakeVectorFromLinearColorRGB(Color=&SelectLinearColorFromArray(
       Array Sampling Mode=::NewEnumerator1
       Direct Array Index=&MakeIntFromFloat(Float=$User.VariantSwitch)
       Color Selection Array=DataInterface() {propertyValues {ColorData [<rgba 0 0.5 0.0749 2>, <rgba 0.1275 0.5 0.0367 2>]}}
       Random Seed=?
      ))
     "Float (Alpha)"=1
    )) {isSetParameters=true, module=SetVariables_#3}
   /SetVariables(Particles.Color003:LinearColor=&MakeLinearColorFromVectorAndFloat(
     "Vector (RGB)"=&MakeVectorFromLinearColorRGB(Color=&SelectLinearColorFromArray(
       Array Sampling Mode=::NewEnumerator1
       Direct Array Index=&MakeIntFromFloat(Float=$User.VariantSwitch)
       Color Selection Array=DataInterface() {propertyValues {ColorData [<rgba 0.7658 1 0 1>, <rgba 0.092 0.827 1 1>]}}
       Random Seed=?
      ))
     "Float (Alpha)"=1
    )) {isSetParameters=true, module=SetVariables_#4}
   /SetVariables(Particles.Color_OutsideGlow:LinearColor=&MakeLinearColorFromVectorAndFloat(
     "Vector (RGB)"=&MakeVectorFromLinearColorRGB(Color=&SelectLinearColorFromArray(
       Array Sampling Mode=::NewEnumerator1
       Direct Array Index=&MakeIntFromFloat(Float=$User.VariantSwitch)
       Color Selection Array=DataInterface() {propertyValues {ColorData [<rgba 0.633 1 0.0746 1>, <rgba 0 0.4061 1 1>]}}
       Random Seed=?
      ))
     "Float (Alpha)"=&Multiply_Float(
      A=&Multiply_Float(
       A=0.3
       B=&FloatFromCurve(FloatCurve=DataInterface() {propertyValues {Curve {keys [
            <at 0.07 0 RCIM_Cubic>
            {arriveTangent=71.4328, interpMode=RCIM_Cubic, leaveTangent=-0.0375, tangentMode=RCTM_Break, time=0.0877, value=1}
            {arriveTangent=-0.2374, interpMode=RCIM_Cubic, tangentMode=RCTM_Break, time=0.1, value=1}
            {arriveTangent=499.1723, interpMode=RCIM_Cubic, leaveTangent=-209.1138, tangentMode=RCTM_Break, time=0.116, value=5}
            <at 0.15 1 RCIM_Cubic>
           ]}}})
      )
      B=&Lerp_Float(A=0.3, B=1.5, Alpha=$System.InterpolateOverTime.InterpolateOverTime_Float.Moving Average)
     )
    )) {isSetParameters=true}
  ]
  rendererClasses [&NiagaraMeshRendererProperties]
  renderers [
   {rendererClass=&NiagaraMeshRendererProperties, rendererIndex=0}
   {rendererClass=&NiagaraMeshRendererProperties, rendererIndex=1}
   {rendererClass=&NiagaraMeshRendererProperties, rendererIndex=2}
   {rendererClass=&NiagaraMeshRendererProperties, rendererIndex=3}
   {rendererClass=&NiagaraMeshRendererProperties, rendererIndex=4}
  ]
 }
 MusicLine003 {
  emitterUpdateScript [
   &EmitterState(
    Inactive Response=::NewEnumerator1
    Life Cycle Mode=::NewEnumerator1
    Loop Behavior=::NewEnumerator1
    Loop Delay=0.4
    Loop Duration=0.35
    UseLoopDelay=true
   )
   &SpawnBurst_Instantaneous(Spawn Count=2, Use Spawn Probability=true, UseLoopCountLimit=false)
  ]
  particleSpawnScript [
   &PartitionParticles()
   &InitializeParticle(
    Color=<rgba 0.0145 0.55 0.4028 1>
    Lifetime=5
    Mesh Scale=&Multiply_VectorByFloat(Vector=<2.8 1 1>, Float=1)
    Mesh Scale Mode=::NewEnumerator4
    Position Offset=&SelectVectorFromArray(
     Direct Array Index=$Particles.PartitionParticles.Partition
     Vector Selection Array=DataInterface() {propertyValues {FloatData [<-2 0 15>, <-2 0 -15>]}}
    )
    UsePositionOffset=true
    Write Material Random=true
   )
   &InitialMeshOrientation(
    Mesh Orientation Mode=::NewEnumerator2
    Rotation=&SelectVectorFromArray(
     Direct Array Index=$Particles.PartitionParticles.Partition
     Vector Selection Array=DataInterface() {propertyValues {FloatData [<0 -0.25 0>, <0 0.25 0>]}}
    )
    Use Rotation Vector=true
   )
   &InitialMeshOrientation(
    Mesh Orientation Mode=::NewEnumerator2
    Rotation=&SelectVectorFromArray(
     Direct Array Index=$Particles.PartitionParticles.Partition
     Vector Selection Array=DataInterface() {propertyValues {FloatData [<0 0 0>, <0 0 0.5>]}}
    )
    Use Rotation Vector=true
   ) {module=InitialMeshOrientation001}
  ]
  particleUpdateScript [
   &ParticleState()
   &DynamicMaterialParameters(
    Index 0 Param 1=&Lerp_Float(
     A=&Lerp_Float(A=0.6, B=1.8, Alpha=&ClampFloat(Float=$System.InterpolateOverTime.InterpolateOverTime_Float.Moving Average))
     B=1
     Alpha=$System.MusicReactivity
    )
    Index 0 Param 4=&FloatFromCurve(
     FloatCurve=DataInterface() {propertyValues {Curve {keys [
         {arriveTangent=2.9441, interpMode=RCIM_Cubic, leaveTangent=2.9441, tangentMode=RCTM_User, time=0, value=0}
         {arriveTangent=11.7012, interpMode=RCIM_Cubic, leaveTangent=11.7012, tangentMode=RCTM_User, time=0.0971, value=0.9641}
        ]}}}
     Scale Curve=150
    )
    Index 1 Param 1=2
    Index 1 Param 2=2
    Index 1 Param 3=&SelectFloatFromArray(
     Direct Array Index=&MakeIntFromFloat(Float=$User.VariantSwitch)
     Float Selection Array=DataInterface() {propertyValues {FloatData [0,1]}}
    )
    Module.Param0WriteEnabled=true
    Module.Param0WriteEnabled1=true
    Module.Param1WriteEnabled=true
    Module.Param1WriteEnabled1=true
    Module.Param2WriteEnabled=false
    Module.Param2WriteEnabled1=true
    Module.Param3WriteEnabled=true
    Module.Param3WriteEnabled1=false
    Write Parameter Index 1=true
   ) {module=DynamicMaterialParameters001}
   /SetVariables(Particles.Scale_Beam:Vector3f=<3 10 1.5>) {isSetParameters=true}
   &ScaleColor(Scale Alpha=0.1, ScaleA=true, ScaleRGB=true)
   /SetVariables(Particles.Color002=&MakeLinearColorFromVectorAndFloat(
     "Vector (RGB)"=&MakeVectorFromLinearColorRGB(Color=&Multiply_LinearColorByFloat(
       Linear Color=<rgba 0 0.2031 0.0304 1>
       Float=&FloatFromCurve(FloatCurve=DataInterface() {propertyValues {Curve {keys [
            {arriveTangent=71.4328, interpMode=RCIM_Cubic, leaveTangent=-0.0375, tangentMode=RCTM_Break, time=0.0517, value=1}
            <at 0.0706 9.9591 RCIM_Cubic>
            {arriveTangent=-0.2374, interpMode=RCIM_Cubic, tangentMode=RCTM_Break, time=0.1, value=1}
            <at 0.2 1 RCIM_Cubic>
           ]}}})
      ))
     "Float (Alpha)"=2
    )) {isSetParameters=true, module=SetVariables_#3}
   /SetVariables(Particles.Color_Beam:LinearColor=&MakeLinearColorFromVectorAndFloat(
     "Vector (RGB)"=&MakeVectorFromLinearColorRGB(Color=&SelectLinearColorFromArray(
       Array Sampling Mode=::NewEnumerator1
       Direct Array Index=&MakeIntFromFloat(Float=$User.VariantSwitch)
       Color Selection Array=DataInterface() {propertyValues {ColorData [<rgba 0 0.2 0.1484 1>, <rgba 0.0853 0.2 0.0089 1>]}}
       Random Seed=?
      ))
     "Float (Alpha)"=1
    )) {isSetParameters=true}
  ]
  rendererClasses [&NiagaraMeshRendererProperties]
  renderers [
   {rendererClass=&NiagaraMeshRendererProperties, rendererIndex=0}
   {rendererClass=&NiagaraMeshRendererProperties, rendererIndex=1}
  ]
 }
 Ring {
  emitterUpdateScript [
   &EmitterState(
    Inactive Response=::NewEnumerator1
    Life Cycle Mode=::NewEnumerator1
    Loop Behavior=::NewEnumerator1
    Loop Delay=0.5
    Loop Duration=0.25
    UseLoopDelay=true
   )
   &SpawnBurst_Instantaneous(Spawn Count=2, Use Spawn Probability=true, UseLoopCountLimit=false)
  ]
  particleSpawnScript [
   &PartitionParticles()
   &InitializeParticle(
    Color=<rgba 1.1198 1.5 0.0448 1>
    Lifetime=2.1
    Mesh Scale Mode=::NewEnumerator3
    Mesh Uniform Scale=0.88
    Position Offset=&SelectVectorFromArray(
     Direct Array Index=$Particles.PartitionParticles.Partition
     Vector Selection Array=DataInterface() {propertyValues {FloatData [<0.65 -3 -16.4>, <0.65 -3 15.2>]}}
    )
    UsePositionOffset=true
    Write Material Random=true
   )
  ]
  particleUpdateScript [
   &ParticleState()
   &UpdateMeshOrientation(Rotation Vector=<0 1 0>)
   &FN_Light_Attributes(
    Color=&ColorFromCurve(ColorCurve=DataInterface() {propertyValues {
       AlphaCurve {keys []}
       BlueCurve {keys [
         <at 0.1741 1>
         {arriveTangent=-2.7461, interpMode=RCIM_Cubic, leaveTangent=-2.7461, time=0.2682, value=0.002}
         <at 0.4088 0.002>
         {arriveTangent=0.3177, interpMode=RCIM_Cubic, leaveTangent=0.3177, time=0.5383, value=0}
         <at 0.8096 0.1739>
         <at 0.8792 1>
        ]}
       GreenCurve {keys [
         <at 0.1741 0.4312>
         {arriveTangent=1.562, interpMode=RCIM_Cubic, leaveTangent=1.562, time=0.2682, value=1}
         <at 0.4088 1>
         <at 0.5383 1 RCIM_Cubic>
         <at 0.8096 1>
         <at 0.8792 0.4312>
        ]}
       RedCurve {keys [
         <at 0.1741 0>
         {arriveTangent=0.1068, interpMode=RCIM_Cubic, leaveTangent=0.1068, time=0.2682, value=0.9978}
         <at 0.4088 0.9978>
         {arriveTangent=-1.8424, interpMode=RCIM_Cubic, leaveTangent=-1.8424, time=0.5383, value=0.0389}
         <at 0.8096 0.0004>
         <at 0.8792 0>
        ]}
      }})
    Create Color Attribute=true
    Create Exponent Attribute=true
    Create Position Attribute=true
    Create Radius Attribute=true
    Exponent=3
    Position=&SelectVectorFromArray(
     Direct Array Index=$Particles.PartitionParticles.Partition
     Vector Selection Array=DataInterface() {propertyValues {FloatData [<0.65 -5 -16.4>, <0.65 -5 15.2>]}}
    )
    Radius=&Multiply_Float(
     A=60
     B=&Lerp_Float(
      A=&FloatFromCurve(
       FloatCurve=DataInterface() {propertyValues {Curve {keys [
           {interpMode=RCIM_Cubic, tangentMode=RCTM_User, time=0.1, value=0}
           {arriveTangent=7.2184, interpMode=RCIM_Cubic, tangentMode=RCTM_Break, time=0.4, value=1}
          ]}}}
       CurveIndex=$System.InterpolateOverTime.InterpolateOverTime_Float.Moving Average
      )
      B=1
      Alpha=$System.MusicReactivity
     )
    )
    Scale Color=15
   )
   &ScaleColor(
    Scale Alpha=&Multiply_Float(
     A=1
     B=&FloatFromCurve(FloatCurve=DataInterface() {propertyValues {Curve {keys [<at 0 0>,<at 0.2 1>]}}})
    )
    ScaleA=true
    ScaleRGB=true
   )
  ]
  rendererClasses [&NiagaraMeshRendererProperties, &NiagaraLightRendererProperties]
  renderers [
   {rendererClass=&NiagaraMeshRendererProperties, rendererIndex=0}
   {rendererClass=&NiagaraLightRendererProperties, rendererIndex=1}
  ]
 }
 glow {
  emitterUpdateScript [
   &EmitterState(
    Inactive Response=::NewEnumerator1
    Life Cycle Mode=::NewEnumerator1
    Loop Behavior=::NewEnumerator1
    Loop Delay=0.5
    Loop Duration=0.5
    UseLoopDelay=true
   )
   &SpawnBurst_Instantaneous(Spawn Count=2, Use Spawn Probability=true, UseLoopCountLimit=false)
  ]
  particleSpawnScript [
   &PartitionParticles()
   &InitializeParticle(
    Lifetime=2
    Position Offset=&SelectVectorFromArray(
     Direct Array Index=$Particles.PartitionParticles.Partition
     Vector Selection Array=DataInterface() {propertyValues {FloatData [<0.65 -3 -16.4>, <0.65 -3 15.2>]}}
    )
    Sprite Size Mode=::NewEnumerator3
    Uniform Sprite Size=40
    UsePositionOffset=true
    Write Material Random=true
   )
   &DynamicMaterialParameters(
    Index 0 Param 1=2
    Index 0 Param 2=10
    Module.Index0Float0Write=true
    Module.Index0Float1Write=true
    Module.Index0Float2Write=true
    Module.Index0Float3Write=true
   )
   &CameraOffset(Camera Offset Amount=10)
  ]
  particleUpdateScript [
   &ParticleState()
   &ScaleColor(
    Scale Alpha=&Multiply_Float(
     A=1
     B=&FloatFromCurve(
      FloatCurve=DataInterface() {propertyValues {Curve {keys [<at 0.07 0 RCIM_Cubic>,<at 0.117 1 RCIM_Cubic>,<at 0.2 0 RCIM_Cubic>]}}}
      Scale Curve=0.1
     )
    )
    Scale RGB=&MakeVectorFromLinearColorRGB(Color=<rgba 0.0452 1 0.2636 1>)
    ScaleA=true
    ScaleRGB=true
   )
  ]
  rendererClasses [&NiagaraSpriteRendererProperties]
  renderers [{rendererClass=&NiagaraSpriteRendererProperties}]
 }
}
summary {
 systemName=NS_Pickaxe_LocalTalent_Flare_Deploy
 emitters [
  {emitterName=MusicLine001, rendererClasses [&NiagaraMeshRendererProperties]}
  {emitterName=MusicLine, rendererClasses [&NiagaraMeshRendererProperties]}
  {emitterName=Ring, rendererClasses [&NiagaraMeshRendererProperties, &NiagaraLightRendererProperties]}
  {emitterName=MusicLine003, rendererClasses [&NiagaraMeshRendererProperties]}
  {emitterName=glow, rendererClasses [&NiagaraSpriteRendererProperties]}
 ]
 userVariables [
  User.bIsFrontEndPreview {type=NiagaraBool}
  User.VariantSwitch {type=NiagaraFloat, defaultValue=0, description=""}
 ]
}
