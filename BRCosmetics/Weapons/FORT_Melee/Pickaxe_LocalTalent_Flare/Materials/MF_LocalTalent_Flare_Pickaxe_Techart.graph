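# Material function for the Local Talent Flare pickaxe that adds audio-reactive effects to incoming material attributes, overriding emissive, base color and opacity mask.
# An audio reactivity function supplies amplitude and spectrum, with an optional sine test amplitude.
# SDF shapes in UV space define speaker, grill, screen and wave falloff regions, alongside a channel-selected tech art mask.
# These shape a speaker glow, an animated waveform and Fresnel-weighted, screen-space stabilized grill dots, layered over the base emission.
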MaterialFunction /BRCosmetics/Weapons/FORT_Melee/Pickaxe_LocalTalent_Flare/Materials/MF_LocalTalent_Flare_Pickaxe_Techart
graph {nodes [
  n1 = GetMaterialAttributes(None=FunctionInput() {inputType=FunctionInput_MaterialAttributes}) {attributeGetTypes [#1, #2, #3]}
  n2 = NamedRerouteUsage()
  n3 = Add(A=@n2)
  n4 = Multiply(A=Multiply(A=Multiply(A=@n3, B=@n2), B=NamedRerouteUsage()), B=NamedRerouteUsage())
  FunctionOutput(None=SetMaterialAttributes!(
    MaterialAttributes=@n1.MaterialAttributes
    EmissiveColor=NamedRerouteUsage()
    OpacityMask=Add(A=@n1.OpacityMask, B=@n4)
    BaseColor=LinearInterpolate(A=@n1.BaseColor, B=@n4, Alpha=@n4)
   ) {attributeSetTypes [#1, #2, #3]})
  n5 = VectorParameter() {group=10_LocalTalentColors, parameterName=DotsDarkColor, sortPriority=0, defaultValue=<rgba 0.4072 0.6445 0.0999 1>}
  n6 = VectorParameter() {group=10_LocalTalentColors, parameterName=DotsColor, sortPriority=0, defaultValue=<rgba 0.8963 0.9473 0.1878 1>}
  n7 = Constant() {r=0.01}
  n8 = Add(
   A=ComponentMask(None=VectorParameter() {group=10_LocalTalent, parameterName=SpeakerGrillUVOffset, defaultValue=<rgba -0.42 -0.205 0 0>}.RGB) {g=true, r=true}
   B=ConstantBiasScale(None=TextureCoordinate())
  )
  n9 = Power(Base=OneMinus(None=MaterialFunctionCall!(UVs=@n8, Radius=@n7) {materialFunction=/FortUILibrary/MaterialFunctions/Shapes/SDFs/MF_SDF_Circle}.SDF)) {constExponent=4}
  n10 = MaterialFunctionCall(UVs=@n8, Radius=Constant() {r=0.155}) {materialFunction=&MF_SDF_Circle}
  n11 = Constant2Vector() {g=1}
  n12 = NamedRerouteDeclaration(None=Saturate(None=If(A=@n10.SDF, "A > B"=@n11.R, "A < B"=@n11.R))) {name=SpeakerGrillMask, nodeColor=<rgba 0 0.2266 1 1>}
  n13 = MaterialFunctionCall!(Position, UVRadialScale) {materialFunction=/Game/Characters/Player/Common/MaterialFunctions/Superlevel/CharacterStabilizedScreenSpaceUVs}
  n14 = ScalarParameter() {defaultValue=100, group=10_LocalTalent, parameterName=Dots Scale, sortPriority=4}
  n15 = TextureObjectParameter() {
   desc=Grayscale pattern texture for effect over the grill.
   group=10_LocalTalent
   parameterName=SquareDots
   samplerType=SAMPLERTYPE_LinearGrayscale
   sortPriority=4
   texture=/BRCosmetics/Weapons/FORT_Melee/Pickaxe_LocalTalent_Flare/Textures/T_LocalTalent_Flare_Pickaxe_Dots
  }
  n16 = TextureSample(
   UVs=Rotator(
    Coordinate=Multiply(A=@n13.StabilizedUVs, B=@n14)
    Time=ScalarParameter() {defaultValue=1.59, group=10_LocalTalent, parameterName=DotsRotation, sortPriority=4}
   )
   Tex=@n15
  ) {
   samplerType=SAMPLERTYPE_LinearGrayscale
   texture=&T_LocalTalent_Flare_Pickaxe_Dots
  }
  n17 = NamedRerouteDeclaration(None=Multiply(A=MaterialFunctionCall() {materialFunction=&MF_Fresnel_FX}.Fresnel, B=@n16.R)) {name=Dots, nodeColor=<rgba 0 1 0.1094 1>}
  n18 = Multiply(
   A=LinearInterpolate(A=@n5.RGB, B=@n6.RGB, Alpha=Multiply(A=@n9, B=@n12)) {constB=0}
   B=Multiply(A=LinearInterpolate(A=@n5.RGB, B=@n6.RGB, Alpha=@n17), B=@n17)
  )
  n19 = ObjectRadius()
  n20 = Distance(A=CameraPositionWS(), B=ObjectPositionWS())
  n21 = Divide(
   A=Divide(A=Divide(A=@n19, B=@n20), B=ViewProperty() {property=MEVP_TanHalfFieldOfView}.Property)
   B=ScalarParameter() {defaultValue=0.1, group=10_LocalTalent, parameterName=Distance Fade Far, sortPriority=1}
  )
  LinearInterpolate(B=@n18, Alpha=ComponentMask(None=Saturate(None=@n21)) {g=true})
  n22 = Subtract(A=PixelDepth().R, B=ScalarParameter() {defaultValue=256, group=10_LocalTalent, parameterName=Wave Static Distance Fade Far, sortPriority=64}) {constB=128}
  n23 = ScalarParameter() {defaultValue=2, group=10_LocalTalent, parameterName=Wave Static Distance Fade Near, sortPriority=64}
  NamedRerouteDeclaration(None=Saturate(None=Divide(A=@n22, B=@n23) {constB=1024})) {name=DistanceFade, nodeColor=<rgba 0 1 0.8125 1>}
  n24 = Constant() {r=0.01}
  n25 = Power(Base=NamedRerouteUsage()) {constExponent=6}
  n26 = LinearInterpolate(
   A=@n24
   B=Multiply(A=ScalarParameter() {defaultValue=0.09, group=10_LocalTalent, parameterName=Wave Max Height, sortPriority=64}, B=@n25)
   Alpha=@n25
  )
  n27 = NamedRerouteUsage()
  NamedRerouteDeclaration(None=LinearInterpolate(B=LinearInterpolate(A=@n26, B=@n24, Alpha=@n27), Alpha=NamedRerouteUsage())) {name=Wave Height, nodeColor=<rgba 1 0.7969 0 1>}
  n28 = Add(
   A=ComponentMask(None=VectorParameter() {group=10_LocalTalent, parameterName=SpeakerUVOffset, defaultValue=<rgba 0.388 0.095 0 0>}.RGB) {g=true, r=true}
   B=TextureCoordinate()
  ) {constB=0}
  n29 = MaterialFunctionCall(UVs=@n28, Radius=Constant() {r=0.2}) {materialFunction=&MF_SDF_Circle}
  n30 = Constant2Vector() {g=1}
  NamedRerouteDeclaration(None=Saturate(None=If(A=@n29.SDF, "A > B"=@n30.R, "A < B"=@n30.R))) {name=SpeakerMask, nodeColor=<rgba 0 0.6953 1 1>}
  n31 = VectorParameter() {group=10_LocalTalentColors, parameterName=SpeakerOuterColor, sortPriority=34, defaultValue=<rgba 0.7682 0.9131 0.1274 1>}
  n32 = LinearInterpolate(
   A=@n31.RGB
   B=VectorParameter() {group=10_LocalTalentColors, parameterName=SpeakerColor, defaultValue=<rgba 0.7682 0.9131 0.1274 1>}.RGB
   Alpha=NamedRerouteUsage()
  )
  n33 = VectorParameter() {group=10_LocalTalentColors, parameterName=SpeakerInnerColor, sortPriority=30, defaultValue=<rgba 1 1 1 1>}
  n34 = Constant() {r=0.07}
  n35 = Constant2Vector() {g=1}
  n36 = MaterialFunctionCall(
   Input=MaterialFunctionCall(UVs=@n28, Radius=@n34) {materialFunction=&MF_SDF_Circle}.SDF
   Input Low=Subtract(B=@n34) {constA=0}
   Input High=@n34
   Target Low=@n35.R
   Target High=@n35.R
  ) {materialFunction=&RemapValueRange}
  n37 = Power(Base=@n36.Result) {constExponent=3}
  n38 = OneMinus(None=NamedRerouteDeclaration(None=Saturate(None=@n37)) {name=SpeakerCirclePow, nodeColor=<rgba 1 0.6563 0 1>})
  n39 = NamedRerouteUsage()
  n40 = Multiply(A=LinearInterpolate(A=@n32, B=Add(A=@n32, B=@n33.RGB), Alpha=Multiply(A=@n38, B=@n39)), B=NamedRerouteUsage())
  n41 = MaterialFunctionCall(UVs=@n28, Radius=Constant() {r=0.045}, Density=Constant() {r=10}) {materialFunction=&RadialGradientExponential}
  n42 = Multiply(A=NamedRerouteUsage(), B=NamedRerouteUsage())
  n43 = VectorParameter() {group=10_LocalTalentColors, parameterName=SpeakerOuterColor, sortPriority=34, defaultValue=<rgba 0.7682 0.9131 0.1274 1>}
  n44 = VectorParameter() {group=10_LocalTalentColors, parameterName=SpeakerColor, defaultValue=<rgba 0.7682 0.9131 0.1274 1>}
  n45 = Constant() {r=0.125}
  n46 = Constant2Vector() {g=1}
  n47 = Power(Base=MaterialFunctionCall(
    Input=MaterialFunctionCall(UVs=@n28, Radius=@n45) {materialFunction=&MF_SDF_Circle}.SDF
    Input Low=Subtract(B=@n45) {constA=0}
    Input High=@n45
    Target Low=@n46.R
    Target High=@n46.R
   ) {materialFunction=&RemapValueRange}.Result) {constExponent=4}
  n48 = NamedRerouteUsage()
  n49 = LinearInterpolate(Alpha=Add(A=ScalarParameter() {group=Audio Reactivity, parameterName=EliminationActive}, B=Power(Base=@n48))) {constA=-1}
  n50 = Multiply(
   A=Multiply(A=LinearInterpolate(A=@n43.RGB, B=@n44.RGB, Alpha=@n47), B=@n47)
   B=LinearInterpolate(Alpha=NamedRerouteDeclaration(None=@n49) {name=SpeakerCurve, nodeColor=<rgba 1 0 0.7031 1>}) {constA=0.4, constB=0.5}
  ) {constB=0.25}
  n51 = If(A=@n42, "A > B"=@n42, "A == B"=@n50, "A < B"=@n50) {
   constB=0.01
   desc="The reskin variant has this emissive ring on the speaker baked into the emissive texture, but the original variant does not"
  }
  n52 = ScalarParameter() {defaultValue=1, group=10_LocalTalent, parameterName=SpeakerGlowBoost}
  NamedRerouteDeclaration(None=Add(A=LinearInterpolate(A=@n40, Alpha=@n41.RadialGradientExponential) {constB=0}, B=Multiply(A=@n51, B=@n52) {constB=1.25})) {name=SpeakerGlow, nodeColor=<rgba 1 0.9844 0 1>}
  n53 = OneMinus(None=@n36.Result)
  n54 = NamedRerouteUsage()
  NamedRerouteDeclaration(None=Saturate(None=Add(A=@n53, B=@n54) {constB=0})) {name=SpeakerCircle, nodeColor=<rgba 0 1 0.9766 1>}
  n55 = Multiply(A=@n17, B=Power(Base=@n17) {constExponent=4})
  NamedRerouteDeclaration(None=Multiply(A=Add(A=@n18, B=@n55), B=ScalarParameter() {defaultValue=2, group=10_LocalTalent, parameterName=Dots Glow Boost, sortPriority=4})) {name=DotsGlow, nodeColor=<rgba 0 0.7187 1 1>}
  NamedRerouteDeclaration(None=@n1.EmissiveColor) {name=EmissionColorIn, nodeColor=<rgba 0.8516 0 1 1>}
  n56 = NamedRerouteDeclaration(None=NamedRerouteUsage()) {name=BaseEmission, nodeColor=<rgba 0.3594 0 1 1>}
  n57 = VectorParameter() {group=10_LocalTalentColors, parameterName=WaveGlowColor, sortPriority=64, defaultValue=<rgba 0.402 1 0.3763 1>}
  n58 = Multiply(
   A=ComponentMask(None=MaterialFunctionCall!(RGB=@n56) {materialFunction=/Engine/Functions/Engine_MaterialFunctions02/Math/RGBtoHSV}."RGB to HSV (0-1)") {b=true}
   B=NamedRerouteDeclaration(None=@n57.RGB) {name=GreenGlowColor, nodeColor=<rgba 1 0.6797 0 1>}
  )
  n59 = NamedRerouteUsage()
  n60 = NamedRerouteUsage()
  n61 = Constant2Vector() {g=0.05, r=-0.05}
  n62 = Constant2Vector() {g=0.5}
  n63 = Saturate(None=OneMinus(None=MaterialFunctionCall(Input=@n60, Input Low=@n61.R, Input High=@n61.R, Target Low=@n62.R, Target High=@n62.R) {materialFunction=&RemapValueRange}.Result))
  n64 = NamedRerouteUsage()
  n65 = Multiply(A=Add(A=Multiply(A=@n59, B=@n63), B=NamedRerouteUsage()), B=LinearInterpolate(Alpha=Power(Base=@n64) {constExponent=6}) {constA=0.8})
  n66 = VectorParameter() {group=10_LocalTalentColors, parameterName=WaveCurveColor, sortPriority=64, defaultValue=<rgba 1 1 1 1>}
  n67 = LinearInterpolate(
   A=@n56
   B=LinearInterpolate(A=LinearInterpolate(A=@n56, B=@n58, Alpha=@n65), B=@n66.RGB, Alpha=NamedRerouteUsage())
   Alpha=NamedRerouteUsage()
  ) {constAlpha=1}
  n68 = NamedRerouteUsage()
  n69 = NamedRerouteUsage()
  n70 = Power(Base=NamedRerouteUsage()) {constExponent=6}
  n71 = Max(A=@n70, B=ScalarParameter() {defaultValue=1, group=10_LocalTalent, parameterName=EnableDotsAlwaysOn, sliderMax=1, sortPriority=0})
  n72 = NamedRerouteDeclaration(None=If(A=@n70, "A > B"=@n70, "A == B"=@n71, "A < B"=@n71)) {name=DotsOn, nodeColor=<rgba 0.2422 0 1 1>}
  n73 = NamedRerouteUsage()
  NamedRerouteDeclaration(None=LinearInterpolate(A=LinearInterpolate(A=@n67, B=@n68, Alpha=@n69), B=NamedRerouteUsage(), Alpha=Multiply(A=@n72, B=@n73))) {name=EmissionColorOut, nodeColor=<rgba 0 1 0.5313 1>}
  n74 = Add(A=ConstantBiasScale(None=TextureCoordinate()), B=Constant2Vector() {r=-0.205})
  n75 = Constant2Vector()
  n76 = Constant2Vector() {g=1}
  n77 = Constant() {r=0.07}
  n78 = Multiply(A=@n77) {constB=0.7}
  n79 = Constant2Vector() {g=1}
  NamedRerouteDeclaration(None=MaterialFunctionCall(
    Input=MaterialFunctionCall!(UVs=@n74, PointA=@n75, PointB=@n76) {materialFunction=/FortUILibrary/MaterialFunctions/Shapes/SDFs/MF_SDF_LineSegment}.SDF
    Input Low=Add(A=@n77, B=@n78)
    Input High=Subtract(A=@n77, B=@n78)
    Target Low=@n79.R
    Target High=@n79.R
   ) {materialFunction=&RemapValueRange}.Result) {name=WaveFalloff, nodeColor=<rgba 0.125 1 0 1>}
  n80 = MaterialFunctionCall(UVs=ConstantBiasScale(None=TextureCoordinate()), PointA=Constant2Vector() {g=0.22, r=0.305}, PointB=Constant2Vector() {g=0.22, r=0.505}) {materialFunction=&MF_SDF_LineSegment}
  n81 = Constant() {r=0.005}
  n82 = Multiply(A=@n81) {constB=4}
  n83 = Add(A=@n81, B=@n82)
  n84 = Subtract(A=@n81, B=@n82)
  n85 = Constant2Vector() {g=1}
  NamedRerouteDeclaration(None=Power(Base=MaterialFunctionCall(Input=@n80.SDF, Input Low=@n83, Input High=@n84, Target Low=@n85.R, Target High=@n85.R) {materialFunction=&RemapValueRange}.Result) {constExponent=1}) {name=ScreenGlow, nodeColor=<rgba 0 0.1328 1 1>}
  n86 = MaterialFunctionCall!(InAmplitude, InSpectrum, CustomFakeSpectrum, CustomFakeAmplitude) {materialFunction=/BRCosmetics/Global/MaterialFunctions/AudioReactivity/MF_AudioReactivity_FrontendLoop}
  NamedRerouteDeclaration(None=@n86.Spectrum) {name=Spectrum, nodeColor=<rgba 0 0.0625 1 1>}
  n87 = VectorParameter() {group=10_LocalTalent, parameterName=Wave UV Offset, sortPriority=64, defaultValue=<rgba 0 0.781 0 0>}
  n88 = ConstantBiasScale(None=Add(A=TextureCoordinate(), B=ComponentMask(None=@n87.RGB) {g=true, r=true})) {bias=-0.5, scale=2}
  n89 = ScalarParameter() {defaultValue=32, group=10_LocalTalent, parameterName=WaveLength}
  n90 = Time()
  n91 = Constant() {r=-1.5}
  n92 = Add(A=Multiply(A=@n88, B=Multiply(A=@n89) {constB=0.5}), B=Frac(None=Multiply(A=@n90, B=@n91) {constB=0.1}))
  n93 = MaterialFunctionCall!(
   UV=@n88
   Wave Height=NamedRerouteUsage()
   Wave Length
   Wave Speed
   GlowMin
   GlowMax
   Outline Thickness=ScalarParameter() {defaultValue=0.01, group=10_LocalTalent, parameterName=Wave Outline Thickness}
   Custom Curve=Sine(None=FunctionOutput(None=@n92) {description=Output the UVs to input in your custom curve logic, outputName=Custom Curve UV, sortPriority=10})
  ) {materialFunction=/BRCosmetics/Characters/Player/Female/Medium/Bodies/F_MED_HangSpec_Screech/Skins/Material_Functions/MF_Wave}
  NamedRerouteDeclaration(None=LinearInterpolate(B=OneMinus(None=@n93.Outline), Alpha=NamedRerouteUsage())) {name=WaveOutline, nodeColor=<rgba 0.4766 0 1 1>}
  NamedRerouteDeclaration(None=Multiply(A=@n93.SDF Outline)) {name=WaveOutlineSDF, nodeColor=<rgba 0 1 0.0859 1>}
  n94 = Sine(None=Time())
  n95 = Constant2Vector() {g=1, r=-1}
  NamedRerouteDeclaration(None=LinearInterpolate(
    A=@n86.Amplitude
    B=MaterialFunctionCall(Input=@n94, Input Low=@n95.R, Input High=@n95.R) {materialFunction=&RemapValue_0-1}.Result
    Alpha=ScalarParameter() {group=10_LocalTalent, parameterName=EnableTestAmplitude, sliderMax=1, sortPriority=0}
   )) {name=Amplitude, nodeColor=<rgba 0 1 0.1094 1>}
  n96 = ChannelMaskParameter(None=FunctionInput() {bUsePreviewValueAsDefault=true, inputName=Tech Art Mask, sortPriority=1}) {group=10_LocalTalent, maskChannel=Blue, parameterName=TechArtMaskChannel, sortPriority=0, defaultValue=<rgba 0 0 1 0>}
  NamedRerouteDeclaration(None=@n96.R) {name=TechArtMask, nodeColor=<rgba 0 1 0.5781 1>}
  n97 = MaterialFunctionCall!(
   UVs=Add(A=TextureCoordinate(), B=Constant2Vector() {g=0.275, r=0.093})
   Size=Constant2Vector() {g=0.12, r=0.25}
   CornerRoundness
   KeepAspectRatio
   PixelSize
  ) {materialFunction=/FortUILibrary/MaterialFunctions/Shapes/SDFs/MF_SDF_Box}
  n98 = Constant2Vector() {g=1}
  NamedRerouteDeclaration(None=Multiply(A=If(A=@n97.SDF, "A > B"=@n98.R, "A < B"=@n98.R), B=NamedRerouteUsage())) {name=ScreenMask, nodeColor=<rgba 0 0.3906 1 1>}
  NamedRerouteUsage()
  Sine(None=Time())
 ]}
