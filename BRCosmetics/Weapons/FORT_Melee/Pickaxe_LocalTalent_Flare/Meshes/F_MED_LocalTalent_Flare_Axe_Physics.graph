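# Physics asset for the Local Talent Flare pickaxe, defining collision and joints for a chain attached to its two-handed handle.
# Three overlapping capsules cover part of the handle, small capsules form the chain links, and a wide flat box caps the last link, dyn_chain_5.
# Five dynamic links hang in series from base_chain with limited swing and zero twist limits, while base_chain's joint to the handle has free swing and twist.
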PhysicsAsset /BRCosmetics/Weapons/FORT_Melee/Pickaxe_LocalTalent_Flare/Meshes/F_MED_LocalTalent_Flare_Axe_Physics
bodies {
 Melee_TwoHanded_Handle [
  {
   shapeType=Capsule
   length=24.5353
   radius=3.7
   shapeName=Melee_TwoHanded_Handle_capsule
   center=<0.0001 7.1685 -8.0967>
   rotation=<rot 4.0244 -90.0001 -0.0007>
  }
  {
   shapeType=Capsule
   length=24.5353
   radius=4.7
   shapeName=Melee_TwoHanded_Handle_capsule
   center=<0.0001 1.1548 -10.7457>
   rotation=<rot 4.0244 -90.0001 -0.0007>
  }
  {
   shapeType=Capsule
   length=24.5353
   radius=3.7
   shapeName=Melee_TwoHanded_Handle_capsule
   center=<0.0001 -6.0078 -8.1357>
   rotation=<rot 4.0244 -90.0001 -0.0007>
  }
 ]
 base_chain [{shapeType=Capsule, length=1, radius=1}]
 dyn_chain_1 [{shapeType=Capsule, length=1, radius=0.8, center=<2 0 0>, rotation=<rot 90 0 0>}]
 dyn_chain_2 [{shapeType=Capsule, length=1, radius=0.8, center=<2 0 0>, rotation=<rot 90 0 0>}]
 dyn_chain_3 [{shapeType=Capsule, length=1, radius=0.8, center=<2 0 0>, rotation=<rot 90 0 0>}]
 dyn_chain_4 [{shapeType=Capsule, length=1, radius=0.8, center=<2 0 0>, rotation=<rot 90 0 0>}]
 dyn_chain_5 [{shapeType=Box, extentX=19, extentY=1.3, extentZ=9.5, shapeName=dyn_chain_5_box, center=<8 0 0>}]
}
constraints [
 {bone1Name=dyn_chain_1, bone2Name=base_chain, swing1LimitDegrees=40, swing2LimitDegrees=30, twistLimitDegrees=0}
 {bone1Name=dyn_chain_2, bone2Name=dyn_chain_1, swing1LimitDegrees=30, swing2LimitDegrees=30, twistLimitDegrees=0}
 {bone1Name=dyn_chain_3, bone2Name=dyn_chain_2, swing1LimitDegrees=30, swing2LimitDegrees=30, twistLimitDegrees=0}
 {bone1Name=dyn_chain_4, bone2Name=dyn_chain_3, swing1LimitDegrees=30, swing2LimitDegrees=30, twistLimitDegrees=0}
 {bone1Name=dyn_chain_5, bone2Name=dyn_chain_4, swing1LimitDegrees=30, swing2LimitDegrees=30, twistLimitDegrees=0}
 {
  bone1Name=base_chain
  bone2Name=Melee_TwoHanded_Handle
  swing1LimitDegrees=45
  swing1Motion=Free
  swing2LimitDegrees=45
  swing2Motion=Free
  twistLimitDegrees=45
  twistMotion=Free
 }
]
