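# Anim Blueprint for the Local Talent Flare pickaxe that adds secondary physics motion to its dangling dyn_chain bones.
# Outside the front end, a RigidBody node simulates relative to Melee_TwoHanded_Handle.
# In the front end, isFE_Pawn (front end and on pawn) selects between two posed five-bone AnimDynamics spring chains with different limits.
# Update also stores nativized gravity-override results, but the AnimGraph's gravity getter is unconnected.
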AnimBlueprint /BRCosmetics/Weapons/FORT_Melee/Pickaxe_LocalTalent_Flare/Meshes/F_MED_LocalTalent_Flare_Axe_AnimBP
target_skeleton=&/F_MED_LocalTalent_Flare_Axe_Skeleton
graphs [
 AnimGraph {nodes [
   "Transform (Modify) Bone - Bone: dyn_chain_5"(Alpha=1, Translation="0,0,0", Rotation="5,0,45", Scale="1,1,1") {
    rotationMode=BMM_Additive
    rotationSpace=BCS_BoneSpace
    translationMode=BMM_Additive
    translationSpace=BCS_BoneSpace
    boneToModify {boneName=dyn_chain_5}
    rotation {pitch=5, roll=45}
   }
   n1 = "Transform (Modify) Bone - Bone: base_chain"(
    ComponentPose=Local To Component(LocalPose=Use cached pose 'StartCache'())
    Alpha=1
    Translation="0,0,0"
    Rotation="-20,-45,90"
    Scale="1,1,1"
   ) {rotationMode=BMM_Additive, rotationSpace=BCS_BoneSpace, boneToModify {boneName=base_chain}, rotation=<rot -20 -45 90>}
   n2 = Anim Dynamics - Bone: dyn_chain_1!(ComponentPose=@n1, bAlphaBoolEnabled:bool=True, ExternalForce="0,0,0") -> (Pose) {
    alphaInputType=Bool
    angularDampingOverride=0.9999
    angularSpringConstant=500
    bAngularSpring=true
    bOverrideAngularDamping=true
    bUseGravityOverride=true
    gravityScale=0.35
    simulationSpace=RootRelative
    alphaBoolBlend {blendInTime=0.2, blendOption=Linear, blendOutTime=0.2}
    boundBone {boneName=dyn_chain_1}
    componentLinearAccScale=<1 1 1>
    componentLinearVelScale=<1 1 1>
    gravityOverride {z=-1280}
    physicsBodyDefinitions [{
      boundBone {boneName=dyn_chain_1}
      boxExtents=<15 10 10>
      constraintSetup {angularLimitsMax=<25 25 28>, angularLimitsMin=<-15 -25 -27>, angularTarget=<1 0 0>}
      localJointOffset=<3 0 0>
     }]
    simSpaceSettings {maxAngularAcceleration=35, maxAngularVelocity=7, simSpaceAngularAlpha=0.7}
   }
   n3 = Anim Dynamics - Bone: dyn_chain_2!(ComponentPose=@n2, bAlphaBoolEnabled:bool=True, ExternalForce="0,0,0") -> (Pose) {
    alphaInputType=Bool
    angularDampingOverride=0.9999
    angularSpringConstant=500
    bAngularSpring=true
    bOverrideAngularDamping=true
    bUseGravityOverride=true
    gravityScale=0.35
    simulationSpace=RootRelative
    alphaBoolBlend {blendInTime=0.2, blendOption=Linear, blendOutTime=0.2}
    boundBone {boneName=dyn_chain_2}
    componentLinearAccScale=<1 1 1>
    componentLinearVelScale=<1 1 1>
    gravityOverride {z=-1280}
    physicsBodyDefinitions [{
      boundBone {boneName=dyn_chain_2}
      boxExtents=<15 10 10>
      constraintSetup {angularLimitsMax=<7 7 15>, angularLimitsMin=<-15 -25 -27>, angularTarget=<1 0 0>}
      localJointOffset=<3 0 0>
     }]
    simSpaceSettings {maxAngularAcceleration=35, maxAngularVelocity=7, simSpaceAngularAlpha=0.7}
   }
   n4 = Anim Dynamics - Bone: dyn_chain_3!(ComponentPose=@n3, bAlphaBoolEnabled:bool=True, ExternalForce="0,0,0") -> (Pose) {
    alphaInputType=Bool
    angularDampingOverride=0.9999
    angularSpringConstant=500
    bAngularSpring=true
    bOverrideAngularDamping=true
    bUseGravityOverride=true
    gravityScale=0.35
    simulationSpace=RootRelative
    alphaBoolBlend {blendInTime=0.2, blendOption=Linear, blendOutTime=0.2}
    boundBone {boneName=dyn_chain_3}
    componentLinearAccScale=<1 1 1>
    componentLinearVelScale=<1 1 1>
    gravityOverride {z=-1280}
    physicsBodyDefinitions [{
      boundBone {boneName=dyn_chain_3}
      boxExtents=<15 10 10>
      constraintSetup {angularLimitsMax=<7 7 15>, angularLimitsMin=<-15 -25 -27>, angularTarget=<1 0 0>}
      localJointOffset=<3 0 0>
     }]
    simSpaceSettings {maxAngularAcceleration=35, maxAngularVelocity=7, simSpaceAngularAlpha=0.7}
   }
   n5 = Anim Dynamics - Bone: dyn_chain_4!(ComponentPose=@n4, bAlphaBoolEnabled:bool=True, ExternalForce="0,0,0") -> (Pose) {
    alphaInputType=Bool
    angularDampingOverride=0.9999
    angularSpringConstant=500
    bAngularSpring=true
    bOverrideAngularDamping=true
    bUseGravityOverride=true
    gravityScale=0.35
    simulationSpace=RootRelative
    alphaBoolBlend {blendInTime=0.2, blendOption=Linear, blendOutTime=0.2}
    boundBone {boneName=dyn_chain_4}
    componentLinearAccScale=<1 1 1>
    componentLinearVelScale=<1 1 1>
    gravityOverride {z=-1280}
    physicsBodyDefinitions [{
      boundBone {boneName=dyn_chain_4}
      boxExtents=<15 10 10>
      constraintSetup {angularLimitsMax=<7 7 15>, angularLimitsMin=<-15 -25 -27>, angularTarget=<1 0 0>}
      localJointOffset=<3 0 0>
     }]
    simSpaceSettings {maxAngularAcceleration=35, maxAngularVelocity=7, simSpaceAngularAlpha=0.7}
   }
   n6 = Anim Dynamics - Bone: dyn_chain_5!(ComponentPose=@n5, bAlphaBoolEnabled:bool=True, ExternalForce="0,0,0") -> (Pose) {
    alphaInputType=Bool
    angularDampingOverride=0.9999
    angularSpringConstant=500
    bAngularSpring=true
    bOverrideAngularDamping=true
    bUseGravityOverride=true
    gravityScale=0.35
    simulationSpace=RootRelative
    alphaBoolBlend {blendInTime=0.2, blendOption=Linear, blendOutTime=0.2}
    boundBone {boneName=dyn_chain_5}
    componentLinearAccScale=<1 1 1>
    componentLinearVelScale=<1 1 1>
    gravityOverride {z=-1280}
    physicsBodyDefinitions [{
      boundBone {boneName=dyn_chain_5}
      boxExtents=<15 10 10>
      constraintSetup {angularLimitsMax=<7 7 15>, angularLimitsMin=<-15 -25 -27>, angularTarget=<1 0 0>}
      localJointOffset=<3 0 0>
     }]
    simSpaceSettings {maxAngularAcceleration=35, maxAngularVelocity=7, simSpaceAngularAlpha=0.7}
   }
   n7 = "Constraint - dyn_chain_5 (R)"(ComponentPose=@n6, Alpha=0.2, ConstraintWeights_0=1) {boneToModify {boneName=dyn_chain_5}, constraintSetup [{transformType=Rotation, targetBone {boneName=b_AR_Root}}], constraintWeights [1]}
   n8 = "Transform (Modify) Bone - Bone: dyn_chain_5"(ComponentPose=@n7, Alpha=1, Translation="0,0,0", Rotation="-12,0,45", Scale="1,1,1") {
    rotationMode=BMM_Additive
    rotationSpace=BCS_BoneSpace
    translationMode=BMM_Additive
    translationSpace=BCS_BoneSpace
    boneToModify {boneName=dyn_chain_5}
    rotation {pitch=-12, roll=45}
   }
   n9 = Component To Local(ComponentPose=@n8)
   n10 = Use cached pose 'StartCache'()
   n11 = Anim Dynamics - Bone: dyn_chain_1(
    ComponentPose="Transform (Modify) Bone - Bone: base_chain"(ComponentPose=Local To Component(LocalPose=@n10), Alpha=1, Translation="0,0,0", Rotation="0,-90,0", Scale="1,1,1") {rotationMode=BMM_Additive, rotationSpace=BCS_BoneSpace, boneToModify {boneName=base_chain}, rotation {yaw=-90}}
    bAlphaBoolEnabled=True
    ExternalForce="0,0,0"
   ) {
    alphaInputType=Bool
    angularDampingOverride=0.9999
    angularSpringConstant=500
    bAngularSpring=true
    bOverrideAngularDamping=true
    bUseGravityOverride=true
    gravityScale=0.35
    simulationSpace=RootRelative
    alphaBoolBlend {blendInTime=0.2, blendOption=Linear, blendOutTime=0.2}
    boundBone {boneName=dyn_chain_1}
    componentLinearAccScale=<1 1 1>
    componentLinearVelScale=<1 1 1>
    gravityOverride {z=-1280}
    physicsBodyDefinitions [{
      boundBone {boneName=dyn_chain_1}
      boxExtents=<25 25 25>
      constraintSetup {angularLimitsMax=<45 5 45>, angularLimitsMin=<-45 -15 -15>, angularTarget=<1 0 0>}
      localJointOffset=<3 0 0>
     }]
    simSpaceSettings {maxAngularAcceleration=35, maxAngularVelocity=7, simSpaceAngularAlpha=0.7}
   }
   n12 = Anim Dynamics - Bone: dyn_chain_2(ComponentPose=@n11, bAlphaBoolEnabled=True, ExternalForce="0,0,0") {
    alphaInputType=Bool
    angularDampingOverride=0.9999
    angularSpringConstant=500
    bAngularSpring=true
    bOverrideAngularDamping=true
    bUseGravityOverride=true
    gravityScale=0.35
    simulationSpace=RootRelative
    alphaBoolBlend {blendInTime=0.2, blendOption=Linear, blendOutTime=0.2}
    boundBone {boneName=dyn_chain_2}
    componentLinearAccScale=<1 1 1>
    componentLinearVelScale=<1 1 1>
    gravityOverride {z=-1280}
    physicsBodyDefinitions [{
      boundBone {boneName=dyn_chain_2}
      boxExtents=<25 25 25>
      constraintSetup {angularLimitsMax=<45 5 45>, angularLimitsMin=<-45 -15 -15>, angularTarget=<1 0 0>}
      localJointOffset=<3 0 0>
     }]
    simSpaceSettings {maxAngularAcceleration=35, maxAngularVelocity=7, simSpaceAngularAlpha=0.7}
   }
   n13 = Anim Dynamics - Bone: dyn_chain_3(ComponentPose=@n12, bAlphaBoolEnabled=True, ExternalForce="0,0,0") {
    alphaInputType=Bool
    angularDampingOverride=0.9999
    angularSpringConstant=500
    bAngularSpring=true
    bOverrideAngularDamping=true
    bUseGravityOverride=true
    gravityScale=0.35
    simulationSpace=RootRelative
    alphaBoolBlend {blendInTime=0.2, blendOption=Linear, blendOutTime=0.2}
    boundBone {boneName=dyn_chain_3}
    componentLinearAccScale=<1 1 1>
    componentLinearVelScale=<1 1 1>
    gravityOverride {z=-1280}
    physicsBodyDefinitions [{
      boundBone {boneName=dyn_chain_3}
      boxExtents=<25 25 25>
      constraintSetup {angularLimitsMax=<45 5 45>, angularLimitsMin=<-45 -15 -15>, angularTarget=<1 0 0>}
      localJointOffset=<3 0 0>
     }]
    simSpaceSettings {maxAngularAcceleration=35, maxAngularVelocity=7, simSpaceAngularAlpha=0.7}
   }
   n14 = Anim Dynamics - Bone: dyn_chain_4(ComponentPose=@n13, bAlphaBoolEnabled=True, ExternalForce="0,0,0") {
    alphaInputType=Bool
    angularDampingOverride=0.9999
    angularSpringConstant=500
    bAngularSpring=true
    bOverrideAngularDamping=true
    bUseGravityOverride=true
    gravityScale=0.35
    simulationSpace=RootRelative
    alphaBoolBlend {blendInTime=0.2, blendOption=Linear, blendOutTime=0.2}
    boundBone {boneName=dyn_chain_4}
    componentLinearAccScale=<1 1 1>
    componentLinearVelScale=<1 1 1>
    gravityOverride {z=-1280}
    physicsBodyDefinitions [{
      boundBone {boneName=dyn_chain_4}
      boxExtents=<25 25 25>
      constraintSetup {angularLimitsMax=<45 5 45>, angularLimitsMin=<-45 -15 -15>, angularTarget=<1 0 0>}
      localJointOffset=<3 0 0>
     }]
    simSpaceSettings {maxAngularAcceleration=35, maxAngularVelocity=7, simSpaceAngularAlpha=0.7}
   }
   n15 = Anim Dynamics - Bone: dyn_chain_5(ComponentPose=@n14, bAlphaBoolEnabled=True, ExternalForce="0,0,0") {
    alphaInputType=Bool
    angularDampingOverride=0.9999
    angularSpringConstant=500
    bAngularSpring=true
    bOverrideAngularDamping=true
    bUseGravityOverride=true
    gravityScale=0.35
    simulationSpace=RootRelative
    alphaBoolBlend {blendInTime=0.2, blendOption=Linear, blendOutTime=0.2}
    boundBone {boneName=dyn_chain_5}
    componentLinearAccScale=<1 1 1>
    componentLinearVelScale=<1 1 1>
    gravityOverride {z=-1280}
    physicsBodyDefinitions [{
      boundBone {boneName=dyn_chain_5}
      boxExtents=<25 25 25>
      constraintSetup {angularLimitsMax=<45 5 45>, angularLimitsMin=<-45 -15 -15>, angularTarget=<1 0 0>}
      localJointOffset=<3 0 0>
     }]
    simSpaceSettings {maxAngularAcceleration=35, maxAngularVelocity=7, simSpaceAngularAlpha=0.7}
   }
   n16 = Component To Local(ComponentPose=@n15)
   n17 = Get isFE_Pawn() -> (isFE_Pawn:bool)
   n18 = RigidBody!(ComponentPose=Local To Component(LocalPose=Use cached pose 'StartCache'()), Alpha, OverrideWorldGravity:Vector="0,0,-980", ExternalForce) -> (Pose) {
    bClampLinearTranslationLimitToRefPose=true
    bOverrideWorldGravity=true
    overridePhysicsAsset=&/F_MED_LocalTalent_Flare_Axe_Physics
    simulationSpace=BaseBoneSpace
    alphaBoolBlend {blendOption=Linear}
    baseBoneRef {boneName=Melee_TwoHanded_Handle}
    componentAppliedLinearAccClamp=<3000 3000 3000>
    simSpaceSettings {
     dampingAlpha=0
     maxAngularAcceleration=2400
     maxAngularVelocity=8
     maxLinearAcceleration=2400
     maxLinearVelocity=700
     worldAlpha=1
     externalLinearDragV=<0.5 0.5 0.5>
    }
   }
   Output Pose(Result=Blend Poses by bool(
     BlendPose_0=Blend Poses by bool(BlendPose_0=@n16, BlendPose_1=@n9, bActiveValue=@n17)
     BlendPose_1=Component To Local(ComponentPose=@n18)
     bActiveValue=Get bIsInFrontEnd()
    )) {layerGroup=None}
   AnimGraphNode_SubInput_0 = Input Pose()
   Save cached pose 'StartCache'(Pose=@AnimGraphNode_SubInput_0) {cacheName=StartCache}
   Get Gravity Override()
  ]}
 EventGraph {nodes [
   Event BlueprintUpdateAnimation()
   > n1 = "Gravity Override State Function (Parameter Asset) (Nativized)"(
    CharacterPartAnimInstance=Self-Reference()
    GameTimeInSeconds=GetGameTimeInSeconds()
    Parameters=&/F_MED_LocalTalent_Flare_Axe_Parameters
   )
   > Set Gravity Override(Gravity Override=@n1)
   > Set RBAN Settings Out(RBAN Settings Out=@n1.OutRigidBodyAnimNodeParameters)
   n2 = AND Boolean(A=Get bIsInFrontEnd(), B=Get bIsOnPawn() -> (bIsOnPawn:bool))
   > Set isFE_Pawn(isFE_Pawn:bool=@n2) -> (Output_Get:bool)
   TryGetPawnOwner()
   Branch(Condition=@n2)
  ]}
]
variables [
 Gravity Override {type=Vector}
 Gravity Override Params {type=GravityOverrideParamsStruct}
 isFE_Pawn {type=boolean}
 RBAN Settings Out {type=FortRigidBodyAnimNodeParameters}
]
